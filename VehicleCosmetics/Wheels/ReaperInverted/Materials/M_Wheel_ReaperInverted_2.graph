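# Unlit material for the ReaperInverted vehicle wheel, building a stylized glowing look from paintable or primary/secondary colors.
# Vertex color channels separate tread, spokes and lug detail, each combining base intensity or normal-mapped Fresnel with its color.
# Scrolling, rotatable noise, masked to spokes by the rim mask and sharpened into hard shapes, adds secondary-colored glow.
# Tire mask and optional LOD rim mask black out areas before the emissive multiplier.
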Material /VehicleCosmetics/Wheels/ReaperInverted/Materials/M_Wheel_ReaperInverted_2
graph {
 properties {
  bAutomaticallySetUsageInEditor=false
  bUsedWithMeshDeformer=true
  bUsedWithSkeletalMesh=true
  shadingModel=MSM_Unlit
  usedShadingModels=Unlit
 }
 nodes [
  Constant()
  n1 = VertexColor()
  n2 = ScalarParameter() {defaultValue=0.5, parameterName=SpokeBaseIntensity}
  n3 = VectorParameter() {parameterName=PaintColor, defaultValue=<rgba 0.65 0 0.19 1>}
  n4 = StaticSwitchParameter(True=@n3.RGB, False=VectorParameter() {parameterName=PrimaryColor, defaultValue=<rgba 1 0.68 0.24 1>}.RGB) {parameterName=Painted}
  n5 = VertexColor()
  n6 = Multiply(
   A=@n5.B
   B=OneMinus(None=TextureSampleParameter2D() {parameterName=Rim_BC, texture=/VehicleCosmetics/Wheels/ReaperInverted/Textures/T_Wheel_ReaperInverted_Rim_BC}.R)
  )
  n7 = OneMinus(None=@n5.B)
  n8 = ScalarParameter() {defaultValue=0.1, parameterName=Lugs_BaseIntensity}
  n9 = Clamp(None=ConstantBiasScale(None=MaterialFunctionCall!(Normal, Exponent) {materialFunction=/VehicleCosmetics/Wheels/Cola_Holographic/Materials/MF_Fresnel_Classic}.Result) {bias=-0.5, scale=3})
  n10 = Multiply(A=Multiply(A=Multiply(A=@n6, B=@n7), B=Add(A=@n8, B=@n9)), B=Constant() {r=10})
  n11 = StaticSwitchParameter(True=@n3.RGB, False=VectorParameter() {parameterName=SecondaryColor, defaultValue=<rgba 0.54 0.65 0.61 1>}.RGB) {parameterName=Painted}
  n12 = ScalarParameter() {defaultValue=0.15, parameterName=Tread_BaseIntensity}
  n13 = LinearInterpolate(A=Constant(), B=@n11, Alpha=VertexColor().R)
  n14 = Max(A=Multiply(A=Multiply(A=@n1.G, B=@n2), B=@n4), B=Max(A=Multiply(A=@n10, B=@n11), B=Multiply(A=@n12, B=@n13)))
  n15 = LinearInterpolate(A=@n14, B=@n14, Alpha=ComponentMask(None=NamedRerouteUsage()) {b=true})
  n16 = TextureSampleParameter2D(UVs=TextureCoordinate()) {parameterName=Tier_N, samplerType=SAMPLERTYPE_Normal, texture=/VehicleCosmetics/Wheels/ReaperInverted/Textures/T_Wheel_ReaperInverted_Tire_N}
  n17 = ConstantBiasScale(None=MaterialFunctionCall(Normal=@n16.RGB) {materialFunction=&MF_Fresnel_Classic}.Result) {bias=-0.3, scale=2}
  n18 = Multiply(
   A=MaterialFunctionCall() {materialFunction=&MF_Fresnel_Classic}.Result
   B=ScalarParameter() {defaultValue=0.15, parameterName=Fresnel_TreadIntensity}
  )
  n19 = ScalarParameter() {defaultValue=0.5, parameterName=Tread Detail Normal Strength}
  n20 = TextureSample(UVs=TextureCoordinate() {uTiling=2, vTiling=2}) {samplerType=SAMPLERTYPE_Normal, texture=/VehicleCosmetics/Wheels/Onetrue_Negative/Textures/Matte_N}
  n21 = Clamp(None=Add(A=Add(A=@n17, B=@n18), B=Multiply(A=@n19, B=@n20.R)))
  n22 = ScalarParameter() {defaultValue=2, parameterName=TreadFresnel_Main}
  n23 = Multiply(A=LinearInterpolate(A=Constant(), B=Multiply(A=@n21, B=@n22), Alpha=VertexColor().R), B=@n11)
  n24 = Constant()
  n25 = TextureSampleParameter2D(UVs=TextureCoordinate()) {parameterName=Rim_N, samplerType=SAMPLERTYPE_Normal, texture=/VehicleCosmetics/Wheels/ReaperInverted/Textures/T_Wheel_ReaperInverted_Rim_N}
  n26 = Constant() {r=0.5}
  n27 = Clamp(None=ConstantBiasScale(None=MaterialFunctionCall(Normal=@n25.RGB, Exponent=@n26) {materialFunction=&MF_Fresnel_Classic}.Result) {bias=-0.5, scale=3})
  n28 = VertexColor()
  n29 = Multiply(
   A=Multiply(A=LinearInterpolate(A=@n24, B=@n27, Alpha=@n28.G), B=ScalarParameter() {defaultValue=1.5, parameterName=SpokeFresnel_Intensity})
   B=@n4
  )
  n30 = Multiply(A=@n29, B=Constant() {r=1})
  n31 = ComponentMask(None=NamedRerouteUsage()) {b=true}
  n32 = TextureCoordinate() {uTiling=0.5, vTiling=0.5}
  n33 = StaticSwitchParameter(True=AppendVector(A=ComponentMask(None=@n32) {g=true}, B=ComponentMask(None=@n32) {r=true}), False=@n32) {parameterName=DetailWarp_NoiseSwapUV}
  n34 = ScalarParameter() {defaultValue=0.5, parameterName=DetailWarp_Noise1Rotation}
  n35 = Time()
  n36 = ScalarParameter() {parameterName=DetailWarp_Noise1Spinner}
  n37 = VectorParameter() {parameterName=DetailWarp_Transform, defaultValue=<rgba 0.5 10 -0.5 0>}
  n38 = Multiply(
   A=Rotator(Coordinate=Rotator(Coordinate=@n33, Time=@n34) {speed=6.2832}, Time=Multiply(A=@n35, B=@n36)) {speed=6.2832}
   B=AppendVector(A=@n37.R, B=@n37.R)
  )
  n39 = VectorParameter() {parameterName=DetailWarp_TransformOffset}
  n40 = Add(A=Multiply(A=Time(), B=AppendVector(A=@n37.B, B=@n37.B)), B=AppendVector(A=@n39.R, B=@n39.R))
  n41 = NamedRerouteUsage()
  n42 = LinearInterpolate(
   A=Constant()
   B=TextureSample(UVs=Add(A=@n38, B=@n40)) {samplerType=SAMPLERTYPE_LinearColor, texture=/VehicleCosmetics/Wheels/_SharedTextures/FX/Universal_Noise_89_RGB}.R
   Alpha=Multiply(A=ComponentMask(None=@n41) {g=true}, B=VertexColor().G)
  )
  n43 = Clamp(None=Multiply(A=ConstantBiasScale(None=@n42) {bias=-0.15, scale=5}, B=Constant() {r=100}))
  n44 = ScalarParameter() {defaultValue=4, parameterName=Shape2_Intensity}
  n45 = Add(A=Add(A=Add(A=@n15, B=@n23), B=LinearInterpolate(A=@n29, B=@n30, Alpha=@n31)), B=Multiply(A=Multiply(A=@n43, B=@n44), B=@n11))
  n46 = VertexColor()
  n47 = LinearInterpolate(
   A=@n45
   B=Constant()
   Alpha=LinearInterpolate(
    B=TextureSample() {samplerType=SAMPLERTYPE_LinearColor, texture=/VehicleCosmetics/Wheels/ReaperInverted/Textures/T_Wheel_ReaperInverted_Rim_M2_LOD}.B
    Alpha=@n46.G
   )
  )
  n48 = Constant()
  n49 = TextureSampleParameter2D(UVs=TextureCoordinate()) {
   parameterName=Tier_M
   samplerType=SAMPLERTYPE_LinearColor
   texture=/VehicleCosmetics/Wheels/ReaperInverted/Textures/T_Wheel_ReaperInverted_Tire_M
  }
  n50 = LinearInterpolate(A=@n48, B=Constant() {r=1}, Alpha=Add(A=@n49.R, B=@n49.R))
  Multiply(
   A=LinearInterpolate(
    A=StaticSwitchParameter(True=@n47, False=@n45) {parameterName=LOds_Mask}
    B=Constant()
    Alpha=LinearInterpolate(A=@n48, B=@n50, Alpha=@n46.R)
   )
   B=ScalarParameter() {defaultValue=1, parameterName=Emissive_Intensity}
  )
  n51 = TextureSampleParameter2D(UVs=TextureCoordinate()) {
   parameterName=Rim_M
   samplerType=SAMPLERTYPE_LinearColor
   texture=/VehicleCosmetics/Wheels/ReaperInverted/Textures/T_Wheel_ReaperInverted_Rim_M
  }
  NamedRerouteDeclaration(None=@n51.RGB) {name=ReaperInverted_M, nodeColor=<rgba 0 0.9766 1 1>}
 ]
}
